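; BGA_CallADropPod_Item is a Blueprint item actor that, judging by its name, belongs to the drop pod call-in gameplay.
; On BeginPlay it sets a four second lifespan, so it destroys itself shortly after spawning.
; The construction script is empty and no other logic is present; any visuals or pod-calling behavior is not defined in this graph.
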(event EventBeginPlay
  (Actor|SetLifeSpan 4.0))

(fn ConstructionScript)
