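; Prototype Tidal Crane drop pod actor that falls, lands, and plays delayed radio voice lines on the instigator for non-authority machines.
; An imminent-destruction component runs only while falling, and projectile movement ticks after it so destruction precedes movement.
; Landing checks Train and DropPod tags to adjust mesh visibility or collision, then the landed state goes net-dormant and removes the danger zone.
; The pod destroys itself when its linked actor is lost or wakes.
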(event EventBeginPlay
  ;[0.04658 0.15 0.03924 0.5] BeginPlay
  (Utilities|FlowControl|Sequence
    (:then_0
      (Networking|SwitchHasAuthority
        (:Remote
          (Utilities|FlowControl|Delay 0.75)
          (Audio|SpawnSoundAttached "/TidalCraneDropPod/Assets/Audio/SC_VO_TidalCraneDropPod_Response.SC_VO_TidalCraneDropPod_Response" (Variables|/Script/Engine.Actor|GetRootComponent (Variables|Actor|GetInstigator)))
          (Utilities|FlowControl|Delay 1.0)
          (Audio|SpawnSoundAttached "/TidalCraneDropPod/Assets/Audio/SC_VO_TidalCraneDropPod_Radio_End.SC_VO_TidalCraneDropPod_Radio_End" (Variables|/Script/Engine.Actor|GetRootComponent (Variables|Actor|GetInstigator))))))
    (:then_1
      (|CallParentFunction))))

(event EventSetStateFalling
  ;[0.15 0.1469 0.03246 0.5] Falling State
  (|CallParentFunction)
  (Class|ImminentDestructionComponent|SetImminentDestructionActive (Variables|Default|GetDropPodImminentCollisionComponent) true))

(event EventOnLanding
  (params (ImpactResult :type Hit%20Result%20Structure))
  ;[0.0497 0.15 0.09982 0.5] Landed
  (|CallParentFunction ImpactResult)
  (CallFunction|CheckLandingActor ImpactResult))

(event EventSetStateLanded
  (bind _dangerzoneactorref (Variables|Runtime|GetDangerZoneActorRef))
  ;[0.0497 0.15 0.09982 0.5] Landed
  (|CallParentFunction)
  (Class|ImminentDestructionComponent|SetImminentDestructionActive (Variables|Default|GetDropPodImminentCollisionComponent) false)
  (Networking|SetNetDormancy "DORM_DormantAll")
  (Utilities|IsValid _dangerzoneactorref
    (:Is%20Valid
      (Actor|DestroyActor _dangerzoneactorref))))

(event Custom|CheckLandingActor
  (params (HitResult :type Hit%20Result%20Structure))
  (bind (_bblockinghit _binitialoverlap _time _distance _location _impactpoint _normal _impactnormal _physmat _hitactor) (Collision|BreakHitResult HitResult))
  (bind _staticmeshcomponent (Variables|Debug|GetStaticMeshComponent))
  ;[0.1424 0.15 0.05003 0.5] Hide the actor if it lands on something we don't want
  (bind _returnvalue (GameplayTags|HasAnyMatchingGameplayTags(Message) _hitactor (GameplayTags|MakeLiteralGameplayTagContainer "(GameplayTags=((TagName=\"Train\"),(TagName=\"TidalCrane.DropPod\")))")))
  (if _returnvalue
    (Development|SetHiddeninGame _staticmeshcomponent true true)
    (Collision|SetCollisionEnabled _staticmeshcomponent "NoCollision")))

(event OnLinkedActorDestroyed(FortLinkToActor)
  (params (DamageCauser :type Actor%20Object%20Reference))
  ;[0.15 0.09126 0.02455 0.5] If the actor I linked was removed, die
  (Actor|DestroyActor))

(event OnLinkedActorDisconnected(FortLinkToActor)
  ;[0.15 0.09126 0.02455 0.5] If the actor I linked was removed, die
  (Actor|DestroyActor))

(event OnLinkedPhysicsObjectWake(FortLinkToActor)
  (params (PhysicsObject :type Actor%20Object%20Reference))
  ;[0.15 0.09126 0.02455 0.5] If the actor I linked was removed, die
  (Actor|DestroyActor))

(fn GetInstigatorAbilitySystemComponent
  (bind _returnvalue (Game|GetInstigator))
  (Utilities|IsValid _returnvalue
    (:Is%20Valid
      (return (Ability|GetFortAbilitySystemComponent _returnvalue)))))

(fn ConstructionScript
  (|CallParentFunction)
  (Components|Tick|AddTickPrerequisiteComponent (Variables|Default|GetFortProjectileMovement) (Variables|Default|GetDropPodImminentCollisionComponent)) ; Try to make sure destruction happens before movement
)
