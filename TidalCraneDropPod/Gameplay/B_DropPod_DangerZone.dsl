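; Drop pod danger zone Blueprint that fades a material effect, likely post-process, in and out as anything overlaps its Capsule.
; Begin overlap plays a timeline from the start, writing its blend weight to the BlendWeight scalar in the global MPC_PPM parameter collection.
; End overlap reverses that timeline from the end.
; Overlapping actors are not filtered, and both transitions restart from a timeline endpoint rather than the current value.
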(event OnComponentBeginOverlap(Capsule)
  (params (OverlappedComponent :type Primitive%20Component%20Object%20Reference) (OtherActor :type Actor%20Object%20Reference) (OtherComp :type Primitive%20Component%20Object%20Reference) (OtherBodyIndex :type Integer) (bFromSweep :type Boolean) (SweepResult :type Hit%20Result%20Structure%20%28by%20ref%29))
  ; BLEND WEIGHT PPM
  (bind (_direction _blendweight) (Timeline|Timeline_BlendWeight PlayFromStart)
    (:Update
      (Rendering|Material|SetScalarParameterValue "/TidalCraneDropPod/DataTables/MPC_PPM.MPC_PPM" "BlendWeight" _blendweight :!class /Script/Engine.KismetMaterialLibrary))))

(event OnComponentEndOverlap(Capsule)
  (params (OverlappedComponent :type Primitive%20Component%20Object%20Reference) (OtherActor :type Actor%20Object%20Reference) (OtherComp :type Primitive%20Component%20Object%20Reference) (OtherBodyIndex :type Integer))
  ; BLEND WEIGHT PPM
  (Timeline|Timeline_BlendWeight ReverseFromEnd))

(fn ConstructionScript)
