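; Projectile that calls in a drop pod where it comes to rest or enters water.
; On either event the server launches a short-lived upward flare, spawns a danger area, then spawns the pod building actor at a configurable height above it, passing in that area actor.
; Non-authority clients play a start voice line on BeginPlay and activation sound and Niagara impact effects on landing, with the water effects limited by a DoOnce.
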(event EventBeginPlay
  (|CallParentFunction)
  (Networking|SwitchHasAuthority
    (:Remote
      (Audio|PlaySoundatLocation "/TidalCraneDropPod/Assets/Audio/SC_VO_TidalCraneDropPod_Start.SC_VO_TidalCraneDropPod_Start" (Transformation|GetActorLocation)))))

(event OnProjectileStop(ProjectileMovementComponent)
  (params (ImpactResult :type Hit%20Result%20Structure%20%28by%20ref%29))
  (bind (_bblockinghit _binitialoverlap _time _distance _location) (Collision|BreakHitResult ImpactResult))
  (bind _returnvalue (Math|Conversions|ToTransform(Vector) _location))
  (Networking|SwitchHasAuthority
    (:Authority
      (bind _returnvalue_2 (Projectile|SpawnProjectile :ProjectileClass "/TidalCraneDropPod/Gameplay/B_Prj_DialADropPod_Flare.B_Prj_DialADropPod_Flare_C" :RequestedBy self :SpawnLocation _location :SpawnRotation "0, 0, 0" :SpawnDirection (Math|Conversions|RotationFromXVector (Math|Vector|VectorUp)) :bUseDefaultPhysics false :InitialSpeed (ScalableFloat|GetValueatLevel (Variables|Default|GetFlareInitialSpeed)) :GravityScale 0.0 :ChargePercent 0.0))
      (Actor|SetLifeSpan _returnvalue_2 7.0)
      (bind _returnvalue_3 (Game|SpawnActorfromClass (Variables|Default|GetDangerAreaClass) _returnvalue "AlwaysSpawn" "MultiplyWithRoot" 0 (Variables|Actor|GetInstigator))) ; Spawn danger area actor
      (Game|SpawnBuildingActorfromClass (Variables|Default|GetDropPodClass) (Math|Transform|MakeTransform (Utilities|Operators|Add (.location _returnvalue) (Math|Vector|MakeVector 0.0 0.0 (ScalableFloat|GetValueatLevel (Variables|Default|GetDropPodSpawnHeight)))) (.rotation _returnvalue) (.scale _returnvalue)) 0 false _returnvalue_3 0 0 (Variables|Actor|GetInstigator)) ; Spawn drop pod
    )
    (:Remote
      (Audio|PlaySoundatLocation "/TidalCraneDropPod/Assets/Audio/SC_TidalCraneDropPod_FlareActivate.SC_TidalCraneDropPod_FlareActivate" _location)
      (Niagara|SpawnSystematLocation (Variables|Default|GetImpact%20FX) _location))))

(event OnEnterWaterEvent(WaterInteractionComponent)
  (params (WaterBody :type Fort%20Water%20Body%20Actor%20Object%20Reference) (WaterInteractionComponent :type Fort%20Water%20Interaction%20Component%20Object%20Reference) (bIsFirstBody :type Boolean))
  (bind _returnvalue (Transformation|GetActorLocation))
  (bind _returnvalue_2 (Transformation|GetActorLocation))
  (bind _returnvalue_3 (Math|Conversions|ToTransform(Vector) _returnvalue))
  (Networking|SwitchHasAuthority
    (:Authority
      (bind _returnvalue_4 (Projectile|SpawnProjectile :ProjectileClass "/TidalCraneDropPod/Gameplay/B_Prj_DialADropPod_Flare.B_Prj_DialADropPod_Flare_C" :RequestedBy self :SpawnLocation (Transformation|GetActorLocation) :SpawnRotation "0, 0, 0" :SpawnDirection (Math|Conversions|RotationFromXVector (Math|Vector|VectorUp)) :bUseDefaultPhysics false :InitialSpeed (ScalableFloat|GetValueatLevel (Variables|Default|GetFlareInitialSpeed)) :GravityScale 0.0 :ChargePercent 0.0))
      (Actor|SetLifeSpan _returnvalue_4 7.0)
      (bind _returnvalue_5 (Game|SpawnActorfromClass (Variables|Default|GetDangerAreaClass) _returnvalue_3 "AlwaysSpawn" "MultiplyWithRoot" 0 (Variables|Actor|GetInstigator))) ; Spawn danger area actor
      (Game|SpawnBuildingActorfromClass (Variables|Default|GetDropPodClass) (Math|Transform|MakeTransform (Utilities|Operators|Add (.location _returnvalue_3) (Math|Vector|MakeVector 0.0 0.0 (ScalableFloat|GetValueatLevel (Variables|Default|GetDropPodSpawnHeight)))) (.rotation _returnvalue_3) (.scale _returnvalue_3)) 0 false _returnvalue_5 0 0 (Variables|Actor|GetInstigator)) ; Spawn drop pod
    )
    (:Remote
      (Utilities|FlowControl|DoOnce false
        (:Completed
          (Audio|PlaySoundatLocation "/TidalCraneDropPod/Assets/Audio/SC_TidalCraneDropPod_FlareActivateInWater.SC_TidalCraneDropPod_FlareActivateInWater" _returnvalue_2)
          (Niagara|SpawnSystematLocation (Variables|Default|GetImpact%20FX) _returnvalue_2))))))

(fn ConstructionScript
  (|CallParentFunction))
