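; Blueprint for the flare projectile in the Dial-A-Drop-Pod gameplay of TidalCraneDropPod.
; On BeginPlay it waits for a delay read from the FlareStopProjectile scalable float.
; It then stops the projectile movement component immediately and destroys the NS_Flare_Tip component, so the flare halts and loses its tip effect.
; The construction script is empty.
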(event EventBeginPlay
  (Utilities|FlowControl|Delay (ScalableFloat|GetValueatLevel (Variables|Default|GetFlareStopProjectile)))
  (Components|Movement|StopMovementImmediately (Variables|Movement|GetProjectileMovementComponent))
  (Components|DestroyComponent (Variables|Default|GetNS_Flare_Tip)))

(fn ConstructionScript)
