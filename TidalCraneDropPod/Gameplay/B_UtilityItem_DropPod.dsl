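; B_UtilityItem_DropPod is a drop pod utility item Blueprint with switchable flare effects.
; Custom events ActivateFlare and DeactivateFlare turn the Idle_Flare_FX component on and off, activating it without a reset.
; The construction script only calls its parent function.
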(event Custom|ActivateFlare
  (Components|Activation|Activate (Variables|Default|GetIdle_Flare_FX) false))

(event Custom|DeactivateFlare
  (Components|Activation|Deactivate (Variables|Default|GetIdle_Flare_FX)))

(fn ConstructionScript
  (|CallParentFunction))
