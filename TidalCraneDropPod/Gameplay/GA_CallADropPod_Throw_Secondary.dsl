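; Gameplay ability for the secondary throw action of the Call A Drop Pod utility item, toggling the item's flare while the ability is active.
; On activation it gets the avatar Fort pawn's current weapon, casts it to B_UtilityItem_DropPod and calls ActivateFlare.
; On end it repeats the lookup and calls DeactivateFlare, whether or not the ability was cancelled.
; Both events call the parent implementation first.
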(event Ability|EventActivateAbility
  (|CallParentFunction)
  (bind _asb_utility_item_drop_pod (Utilities|Casting|CastToB_UtilityItem_DropPod (Weapon|GetCurrentWeapon (CallFunction|GetAvatarasFortPlayerPawn))))
  (Class|BUtilityItemDropPod|ActivateFlare _asb_utility_item_drop_pod))

(event Ability|EventOnEndAbility
  (params (bWasCancelled :type Boolean))
  (|CallParentFunction bWasCancelled)
  (bind _asb_utility_item_drop_pod (Utilities|Casting|CastToB_UtilityItem_DropPod (Weapon|GetCurrentWeapon (CallFunction|GetAvatarasFortPlayerPawn))))
  (Class|BUtilityItemDropPod|DeactivateFlare _asb_utility_item_drop_pod))
